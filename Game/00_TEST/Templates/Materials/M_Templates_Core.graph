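# M_Templates_Core is an unlit translucent test template material for Niagara sprites, built on a scorch-hole decal texture.
# The texture's red channel, raised by a power and halved, blends particle color toward a desaturated version, so brighter areas lose saturation.
# That blended color is multiplied by the red channel, and a separate product of red and particle alpha forms a mask, with no output connections shown.
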Material /Game/00_TEST/Templates/Materials/M_Templates_Core
graph {
 properties {
  bAutomaticallySetUsageInEditor=false
  bUsedWithNiagaraSprites=true
  blendMode=BLEND_Translucent
  shadingModel=MSM_Unlit
  usedShadingModels=Unlit
 }
 nodes [
  n1 = ParticleColor()
  n2 = TextureSample() {texture=/Game/Effects/Fort_Effects/Gameplay/SmokeDustExplosions/Textures/T_Decal_Scorch_Hole}
  n3 = Power(Base=@n2.R)
  Multiply(A=LinearInterpolate(A=@n1.RGB, B=Desaturation(None=@n1.RGB), Alpha=Multiply(A=@n3) {constB=0.5}), B=@n2.R)
  Multiply(A=@n2.R, B=@n1.A)
 ]
}
