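# Translucent template material for Niagara ribbons, located in a test folder, that draws a soft smoke trail tinted by particle color.
# The red channel of the smoke trail texture is raised to a power and multiplied by particle RGB.
# The same channel is multiplied by particle alpha and a fade mask built from two opposed smoothsteps on UV1's U coordinate, tapering both ribbon ends asymmetrically.
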Material /Game/00_TEST/Templates/Materials/M_Templates_Ribbon
graph {
 properties {bAutomaticallySetUsageInEditor=false, bUsedWithNiagaraRibbons=true, blendMode=BLEND_Translucent, usedShadingModels=Default Lit}
 nodes [
  n1 = TextureSample() {texture=/Game/Effects/Fort_Effects/Textures/Smoke/T_Smoke_Trail_Soft_Chromatic}
  n2 = ComponentMask(None=TextureCoordinate() {coordinateIndex=1}) {r=true}
  n3 = Saturate(None=SmoothStep(Value=@n2) {constMax=0.2})
  n4 = Saturate(None=SmoothStep(Value=OneMinus(None=@n2)) {constMax=0.8})
  n5 = ParticleColor()
  Multiply(A=Multiply(A=@n1.R, B=Multiply(A=@n3, B=@n4)), B=@n5.A)
  Multiply(A=Power(Base=@n1.R), B=@n5.RGB)
 ]
}
