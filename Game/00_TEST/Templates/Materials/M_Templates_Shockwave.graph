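# M_Templates_Shockwave is a translucent, unlit template material for Niagara mesh particles, named for shockwave effects.
# Particle color RGB and alpha are each multiplied by a shared Fresnel term, weighting both toward the mesh's silhouette edges.
# It has no textures or parameters, so the particle system drives the tint and fade.
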Material /Game/00_TEST/Templates/Materials/M_Templates_Shockwave
graph {
 properties {
  bAutomaticallySetUsageInEditor=false
  bUsedWithNiagaraMeshParticles=true
  blendMode=BLEND_Translucent
  shadingModel=MSM_Unlit
  usedShadingModels=Unlit
 }
 nodes [
  n1 = ParticleColor()
  n2 = Fresnel()
  Multiply(A=@n1.A, B=@n2)
  Multiply(A=@n1.RGB, B=@n2)
 ]
}
