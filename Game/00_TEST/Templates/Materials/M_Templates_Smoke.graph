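# Translucent, lit smoke material template for Niagara sprites, with volumetric directional lighting and an optional fire glow.
# Cloud texture UVs scale from center over particle lifetime with a random offset per particle.
# A noise-driven sphere mask shapes the puff, feeding grayscale color and contrast-adjusted opacity, both tinted by particle color.
# A sphere normal blends with flattened cloud normals, while a dynamic parameter scales fire-colored emissive in the dense core.
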Material /Game/00_TEST/Templates/Materials/M_Templates_Smoke
graph {
 properties {
  bAutomaticallySetUsageInEditor=false
  bUsedWithNiagaraSprites=true
  blendMode=BLEND_Translucent
  translucencyLightingMode=TLM_VolumetricDirectional
  usedShadingModels=Default Lit
 }
 nodes [
  n1 = ScalarParameter() {defaultValue=2.3, parameterName=Scale Min}
  n2 = ScalarParameter() {defaultValue=6, parameterName=Scale Max}
  n3 = ParticleRelativeTime()
  n4 = Add(
   A=MaterialFunctionCall(Texture Scale=LinearInterpolate(A=@n1, B=@n2, Alpha=@n3) {constA=2, constB=4}) {materialFunction=&ScaleUVsByCenter}.UVs
   B=ParticleRandom()
  )
  MaterialFunctionCall(
   BaseNormal=Multiply(
    A=TextureSample() {samplerType=SAMPLERTYPE_Normal, texture=&T_Sphere_N}.RGB
    B=Constant3Vector() {constant=<rgba 3 3 1 0>}
   )
   AdditionalNormal=MaterialFunctionCall(
    Normal=TextureSample(UVs=@n4) {samplerType=SAMPLERTYPE_Normal, texture=/Game/Effects/Fort_Effects/Textures/Smoke/T_Clouds_01_M}.RGB
    Flatness=ScalarParameter() {defaultValue=0.75, parameterName=Flatness}
   ) {materialFunction=&FlattenNormal}.Result
  ) {materialFunction=&BlendAngleCorrectedNormals}
  n5 = ParticleColor()
  n6 = TextureSample(UVs=@n4) {texture=/Game/Effects/Fort_Effects/Textures/Smoke/T_Clouds_01}
  n7 = LinearInterpolate(A=ScalarParameter() {defaultValue=0.2, parameterName=Remap}, Alpha=@n6.R)
  n8 = Multiply(A=SphereMask(A=TextureCoordinate(), B=Constant() {r=0.5}, Radius=Multiply(A=@n7) {constB=0.5}) {hardnessPercent=0}, B=@n6.R)
  n9 = Multiply(A=@n8, B=ScalarParameter() {defaultValue=2})
  Multiply(A=@n5.RGB, B=LinearInterpolate(A=ScalarParameter() {defaultValue=0.75, parameterName=Min Grayscale}, Alpha=@n9) {constA=0.25})
  Multiply(
   A=@n5.A
   B=MaterialFunctionCall(In=@n9, Contrast=ScalarParameter() {defaultValue=0.5, parameterName=Contrast}) {materialFunction=&CheapContrast}.Result
  )
  Multiply(
   A=Power(Base=@n8) {constExponent=3}
   B=Multiply(
    A=DynamicParameter() {defaultValue=<rgba 0 0 0 1>, paramNames [Emissive,-UNUSED-,-UNUSED-,-UNUSED-]}.Emissive
    B=VectorParameter() {parameterName=Fire Color, defaultValue=<rgba 200 15.9639 0 1>}.RGB
   )
  )
  TextureSample()
 ]
}
