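# Unlit translucent template material for Niagara sprites that draws a soft round glow tinted by particle color.
# A zero-hardness SphereMask centered on the UVs creates a procedural radial falloff without textures.
# Power and SmoothStep shape that falloff, which then multiplies particle RGB and particle alpha separately.
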Material /Game/00_TEST/Templates/Materials/M_Templates_SoftGlow
graph {
 properties {
  bAutomaticallySetUsageInEditor=false
  bUsedWithNiagaraSprites=true
  blendMode=BLEND_Translucent
  shadingModel=MSM_Unlit
  usedShadingModels=Unlit
 }
 nodes [
  n1 = ParticleColor()
  n2 = TextureCoordinate()
  n3 = Constant() {r=0.5}
  n4 = SmoothStep(Value=Power(Base=SphereMask(A=@n2, B=@n3, Radius=@n3) {hardnessPercent=0}))
  Multiply(A=@n1.RGB, B=@n4)
  Multiply(A=@n1.A, B=@n4)
 ]
}
