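# M_Templates_Streaks is an unlit translucent material for Niagara sprites that draws explosion streaks, stored among test templates.
# MF_SubUV samples the streak flipbook texture with manual time driven by ParticleRandom, so the frame comes from a per-particle random value rather than animation.
# The red channel is multiplied separately by particle color RGB and alpha, letting the emitter control tint and fade.
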Material /Game/00_TEST/Templates/Materials/M_Templates_Streaks
graph {
 properties {
  bAutomaticallySetUsageInEditor=false
  bUsedWithNiagaraSprites=true
  blendMode=BLEND_Translucent
  shadingModel=MSM_Unlit
  usedShadingModels=Unlit
 }
 nodes [
  n1 = Constant2Vector() {g=1, r=4}
  n2 = ComponentMask(None=MaterialFunctionCall!(
    U_Images=@n1.R
    V_Images=@n1.R
    SubUV Texture RGB=TextureObject() {texture=/Game/Effects/Fort_Effects/Gameplay/SmokeDustExplosions/Textures/T_Explosion_Streak_02}
    Time Speed
    Manual Time Control=StaticBool() {value=true}
    Time=ParticleRandom()
    UVs
   ) {materialFunction=/Game/Effects/Fort_Effects/Materials/_Functions/MF_SubUV}.RGB) {r=true}
  n3 = ParticleColor()
  Multiply(A=@n2, B=@n3.RGB)
  Multiply(A=@n2, B=@n3.A)
 ]
}
