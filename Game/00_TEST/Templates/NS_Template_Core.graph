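# NS_Template_Core is a template Niagara system for a brief, bright HDR orange core flash, likely for explosions or impacts.
# Its single Core emitter bursts a couple of short-lived particles, drawn by both sprite and light renderers.
# User.Radius, multiplied by a random factor, sets sprite size, so one parameter scales the sprite.
# Over each particle's life, curves grow the sprite and fade its alpha.
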NiagaraSystem /Game/00_TEST/Templates/NS_Template_Core
dependencies {usedDataInterfaces [&NiagaraDataInterfaceCurve], usedDynamicInputs [], usedModules [&SystemState]}
emitters {Core {
  emitterUpdateScript [
   &EmitterState(Life Cycle Mode=::NewEnumerator1, Loop Behavior=::NewEnumerator1, Loop Duration=1, UseLoopDelay=false)
   &SpawnBurst_Instantaneous(Spawn Count=2, Use Spawn Probability=false, UseLoopCountLimit=false)
  ]
  particleSpawnScript [&InitializeParticle(
    Color=<rgba 50 8.8141 0 1>
    Lifetime=0.2
    Mass Max=2
    Mass Mode=::NewEnumerator3
    Sprite Rotation Mode=::NewEnumerator2
    Sprite Size Mode=::NewEnumerator3
    Uniform Sprite Size=&Multiply_Float(A=$User.Radius, B=&RandomRangeFloat(Minimum=1.5, Maximum=2))
    UsePositionOffset=false
    Write Material Random=true
   )]
  particleUpdateScript [
   &ParticleState()
   &ScaleSpriteSize(Uniform Curve Sprite Scale=DataInterface() {propertyValues {Curve {keys [{tangentMode=RCTM_User, time=0, value=0.5756}, {tangentMode=RCTM_User, time=1, value=1.1792}]}}})
   &ScaleColor(
    Scale Alpha=&FloatFromCurve(FloatCurve=DataInterface() {propertyValues {Curve {keys [<at -0.0046 1.0113 RCIM_Cubic>,<at 1.0011 0.002 RCIM_Cubic>]}}})
    ScaleA=true
    ScaleRGB=true
   )
  ]
  rendererClasses [&NiagaraSpriteRendererProperties, &NiagaraLightRendererProperties]
  renderers [
   {rendererClass=&NiagaraSpriteRendererProperties}
   {rendererClass=&NiagaraLightRendererProperties, rendererIndex=1}
  ]
 }}
summary {
 systemName=NS_Template_Core
 emitters [{emitterName=Core, rendererClasses [&NiagaraSpriteRendererProperties, &NiagaraLightRendererProperties]}]
 userVariables [User.Radius {type=NiagaraFloat, defaultValue=256, description=""}]
}
