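# Niagara template system for a brief HDR orange flash, built from a single sprite emitter named Flash.
# An instantaneous burst spawns short-lived sprites sized as a multiple of the exposed User.Radius parameter, so one value scales the effect.
# A curve-driven ScaleColor fades alpha from full to near zero over each particle's quarter-second life.
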NiagaraSystem /Game/00_TEST/Templates/NS_Template_Flash
dependencies {usedDataInterfaces [&NiagaraDataInterfaceCurve], usedDynamicInputs [], usedModules [&SystemState]}
emitters {Flash {
  emitterUpdateScript [
   &EmitterState(Life Cycle Mode=::NewEnumerator1, Loop Behavior=::NewEnumerator1, Loop Duration=1, UseLoopDelay=false)
   &SpawnBurst_Instantaneous(Use Spawn Probability=false, UseLoopCountLimit=false)
  ]
  particleSpawnScript [&InitializeParticle(
    Color=<rgba 3 0.5288 0 1>
    Lifetime=0.25
    Mass Max=2
    Mass Mode=::NewEnumerator3
    Sprite Size Mode=::NewEnumerator3
    Uniform Sprite Size=&Multiply_Float(A=$User.Radius, B=3)
    UsePositionOffset=false
    Write Material Random=true
   )]
  particleUpdateScript [
   &ParticleState()
   &ScaleColor(
    Scale Alpha=&FloatFromCurve(FloatCurve=DataInterface() {propertyValues {Curve {keys [<at -0.0046 1.0113 RCIM_Cubic>,<at 1.0011 0.002 RCIM_Cubic>]}}})
    ScaleA=true
    ScaleRGB=true
   )
  ]
  rendererClasses [&NiagaraSpriteRendererProperties]
  renderers [{rendererClass=&NiagaraSpriteRendererProperties}]
 }}
summary {
 systemName=NS_Template_Flash
 emitters [{emitterName=Flash, rendererClasses [&NiagaraSpriteRendererProperties]}]
 userVariables [User.Radius {type=NiagaraFloat, defaultValue=256, description=""}]
}
