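# Niagara template system for a short orange shockwave, sized by a User.Radius parameter.
# A single Shockwave emitter burst-spawns short-lived mesh particles with HDR color and uniform mesh scale derived from Radius divided by a constant.
# Over each particle's life, a curve grows the mesh from partial to full size while an alpha curve fades it out.
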NiagaraSystem /Game/00_TEST/Templates/NS_Template_Shockwave
dependencies {usedDataInterfaces [&NiagaraDataInterfaceCurve], usedDynamicInputs [], usedModules [&SystemState]}
emitters {Shockwave {
  emitterUpdateScript [
   &EmitterState(Life Cycle Mode=::NewEnumerator1, Loop Behavior=::NewEnumerator1, Loop Duration=1, UseLoopDelay=false)
   &SpawnBurst_Instantaneous(Use Spawn Probability=false, UseLoopCountLimit=false)
  ]
  particleSpawnScript [&InitializeParticle(
    Color=<rgba 30 5.2885 0 1>
    Lifetime=0.2
    Mass Max=2
    Mass Mode=::NewEnumerator3
    Mesh Scale Mode=::NewEnumerator3
    Mesh Uniform Scale=&Divide_Float(A=$User.Radius, B=75)
    UsePositionOffset=false
    Write Material Random=true
   )]
  particleUpdateScript [
   &ParticleState()
   &ScaleMeshSize(Scale Factor=&VectorFromFloat(Value=&FloatFromCurve(FloatCurve=DataInterface() {propertyValues {Curve {
         defaultValue=1
         keys [
          {arriveTangent=1.2949, interpMode=RCIM_Cubic, leaveTangent=1.2949, tangentMode=RCTM_User, time=0, value=0.4}
          {arriveTangent=0.0025, interpMode=RCIM_Cubic, leaveTangent=0.0025, tangentMode=RCTM_User, time=1, value=1}
         ]
        }}})))
   &ScaleColor(
    Scale Alpha=&FloatFromCurve(FloatCurve=DataInterface() {propertyValues {Curve {keys [
         {arriveTangent=-1.7785, interpMode=RCIM_Cubic, leaveTangent=-1.7785, tangentMode=RCTM_User, time=-0.0046, value=1.0113}
         <at 1.0011 0.002 RCIM_Cubic>
        ]}}})
    ScaleA=true
    ScaleRGB=true
   )
  ]
  rendererClasses [&NiagaraMeshRendererProperties]
  renderers [{rendererClass=&NiagaraMeshRendererProperties, rendererIndex=0}]
 }}
summary {
 systemName=NS_Template_Shockwave
 emitters [{emitterName=Shockwave, rendererClasses [&NiagaraMeshRendererProperties]}]
 userVariables [User.Radius {type=NiagaraFloat, defaultValue=256, description=""}]
}
