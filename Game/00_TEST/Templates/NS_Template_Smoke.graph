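# Template Niagara smoke effect that bursts a handful of large grey sprites from a single emitter.
# Particles spawn in a sphere scaled from User.Radius, receive random velocity and slow under linear drag.
# Over each lifetime, curves quickly grow the sprites, lighten their color from dark and fade alpha.
# A dynamic material parameter drops from one to zero early in life, alongside a per-particle material random value.
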NiagaraSystem /Game/00_TEST/Templates/NS_Template_Smoke
dependencies {
 usedDataInterfaces [&NiagaraDataInterfaceCurve, &NiagaraDataInterfaceColorCurve]
 usedDynamicInputs []
 usedModules [&SystemState]
}
emitters {Smoke {
  emitterUpdateScript [
   &EmitterState(Life Cycle Mode=::NewEnumerator1, Loop Behavior=::NewEnumerator1, Loop Duration=1, UseLoopDelay=false)
   &SpawnBurst_Instantaneous(Spawn Count=8, Use Spawn Probability=false, UseLoopCountLimit=false)
  ]
  particleSpawnScript [
   &InitializeParticle(
    Color=<rgba 0.3542 0.3542 0.3542 1>
    Lifetime Min=1.2
    Lifetime Mode=::NewEnumerator1
    Mass Max=2
    Mass Mode=::NewEnumerator3
    Sprite Size Mode=::NewEnumerator1
    Uniform Sprite Size Max=500
    Uniform Sprite Size Min=300
    UsePositionOffset=false
    Write Material Random=true
   )
   &ShapeLocation(
    MaskSpawnGroup=false
    Override Seed=false
    Sphere Radius=&Divide_Float(A=$User.Radius, B=10)
    UseOwnerScale=true
   )
   &AddVelocity(
    Module.Use Falloff Distance=false
    Override Seed=false
    Velocity Mode=::NewEnumerator1
    Velocity Speed=&RandomRangeFloat(Minimum=100, Maximum=300)
   )
  ]
  particleUpdateScript [
   &ParticleState()
   &ScaleSpriteSize(Uniform Curve Sprite Scale=DataInterface() {propertyValues {Curve {keys [
        {arriveTangent=5.4675, interpMode=RCIM_Cubic, leaveTangent=5.4675, tangentMode=RCTM_User, time=0, value=0.1}
        {arriveTangent=3.3641, interpMode=RCIM_Cubic, leaveTangent=3.3641, tangentMode=RCTM_User, time=0.0734, value=0.6502}
        {arriveTangent=0.5482, interpMode=RCIM_Cubic, leaveTangent=0.5482, tangentMode=RCTM_User, time=0.2516, value=0.9319}
        <at 0.6852 1 RCIM_Cubic>
       ]}}})
   &Drag(Drag=&RandomRangeFloat(Minimum=2, Maximum=3), Use Linear Drag=true, Use Rotational Drag=false)
   &ScaleColor(
    Scale Alpha=&FloatFromCurve(FloatCurve=DataInterface() {propertyValues {Curve {keys [<at -0.0046 1.0113 RCIM_Cubic>,<at 1.0011 0.002 RCIM_Cubic>]}}})
    Scale RGB=&MakeVectorFromLinearColorRGB(Color=&ColorFromCurve(ColorCurve=DataInterface() {propertyValues {
        AlphaCurve {keys [<at 0 1 RCIM_Cubic>]}
        BlueCurve {keys [<at 0.1224 0.1094>,<at 1 1 RCIM_Cubic>]}
        GreenCurve {keys [<at 0.1224 0.1094>,<at 1 1 RCIM_Cubic>]}
        RedCurve {keys [<at 0.1224 0.1094>,<at 1 1 RCIM_Cubic>]}
       }}))
    ScaleA=true
    ScaleRGB=true
   )
   &SolveForcesAndVelocity(Clamp Velocity=false, Limit Acceleration=false)
   &DynamicMaterialParameters(
    Index 0 Param 1=&FloatFromCurve(FloatCurve=DataInterface() {propertyValues {Curve {keys [{arriveTangent=0.198, interpMode=RCIM_Cubic, leaveTangent=0.198, tangentMode=RCTM_User, time=0, value=1}, <at 0.2022 0>]}}})
    Module.Index0Float0Write=true
    Module.Index0Float1Write=true
    Module.Index0Float2Write=true
    Module.Index0Float3Write=true
   )
  ]
  rendererClasses [&NiagaraSpriteRendererProperties]
  renderers [{rendererClass=&NiagaraSpriteRendererProperties}]
 }}
summary {
 systemName=NS_Template_Smoke
 emitters [{emitterName=Smoke, rendererClasses [&NiagaraSpriteRendererProperties]}]
 userVariables [User.Radius {type=NiagaraFloat, defaultValue=256, description=""}]
}
